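# Masked, lit Niagara ribbon material for the ink-textured bits of the MissusMind skydiving contrail.
# Independently panned ink samples provide color, alpha and erosion masks, with random particle UV offsets and stepped opacity.
# Panning body normals drive a fake sun glint from sky light direction and blend into surface normals.
# Optional emissive uses a normal-distorted reflection texture; near-camera fade is bypassed on Low and Medium quality.
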Material /BRCosmetics/Athena/Effects/Particles/Skydiving/NS_Contrail_MissusMind/M_MissusMind_Bits_Fixed
graph {
 properties {
  bAutomaticallySetUsageInEditor=false
  bUsedWithNiagaraRibbons=true
  blendMode=BLEND_Masked
  translucencyLightingMode=TLM_Surface
  usedShadingModels=Default Lit
 }
 nodes [
  NamedRerouteDeclaration(None=Multiply(A=NamedRerouteUsage(), B=NamedRerouteUsage())) {name=BaseColor, nodeColor=<rgba 0 0 0 1>}
  NamedRerouteDeclaration(None=ScalarParameter() {defaultValue=0.25, parameterName=Roughness}) {name=Roughness, nodeColor=<rgba 0 0 0 1>}
  n1 = Step(Y=OneMinus(None=NamedRerouteUsage()), X=NamedRerouteUsage())
  n2 = NamedRerouteUsage()
  NamedRerouteDeclaration(None=Multiply(A=Multiply(A=@n1, B=@n2), B=NamedRerouteUsage())) {name=Opacity, nodeColor=<rgba 0 0 0 1>}
  n3 = Power(Base=MaterialFunctionCall(Distance=ScalarParameter() {defaultValue=75, parameterName=NearFadeDistance}) {materialFunction=&Near_Camera_Fade}.Result)
  n4 = Constant() {r=1}
  NamedRerouteDeclaration(None=QualitySwitch(Default=@n3, Low=@n4, High=@n3, Medium=@n4, Epic=@n3)) {name=CameraFade, nodeColor=<rgba 1 0 0.7734 1>}
  n5 = Multiply(
   A=TextureCoordinate()
   B=AppendVector(A=ScalarParameter() {defaultValue=1.5, parameterName=Tiling_U}, B=ScalarParameter() {defaultValue=1, parameterName=Tiling_V})
  )
  n6 = ScalarParameter() {defaultValue=-1, parameterName=PanSpeed3_U}
  n7 = ScalarParameter() {parameterName=PanSpeed3_V}
  NamedRerouteDeclaration(None=Panner(Coordinate=@n5, Speed=AppendVector(A=@n6, B=@n7)) {speedX=-1}) {name=UVs_PanningMask, nodeColor=<rgba 0 0.2266 1 1>}
  n8 = AppendVector(A=ScalarParameter() {defaultValue=-3.5, parameterName=PanSpeed_U}, B=ScalarParameter() {defaultValue=0.5, parameterName=PanSpeed_V})
  NamedRerouteDeclaration(None=Add(A=Panner(Coordinate=@n5, Speed=@n8) {speedX=-1, speedY=0.5}, B=NamedRerouteUsage())) {name=UVs_PanningErosion, nodeColor=<rgba 0.5469 1 0 1>}
  n9 = TextureSample(UVs=NamedRerouteUsage()) {samplerType=SAMPLERTYPE_Masks, texture=&/T_FX_MissusMind_Ink_M}
  NamedRerouteDeclaration(None=@n9.B) {name=Mask_Color, nodeColor=<rgba 1 0 0.8672 1>}
  NamedRerouteDeclaration(None=TextureSample(UVs=NamedRerouteUsage()) {samplerType=SAMPLERTYPE_Masks, texture=&T_FX_MissusMind_Ink_M}.G) {name=Mask_Erosion, nodeColor=<rgba 0 1 0.8125 1>}
  NamedRerouteDeclaration(None=ScalarParameter() {defaultValue=0.65, parameterName=Specular}) {name=Specular, nodeColor=<rgba 0 0 0 1>}
  NamedRerouteDeclaration(None=@n9.R) {name=Mask_Alpha, nodeColor=<rgba 1 0.4922 0 1>}
  n10 = ParticleColor()
  NamedRerouteDeclaration(None=@n10.RGB) {name=ParticleColor, nodeColor=<rgba 1 0 0.2344 1>}
  NamedRerouteDeclaration(None=DynamicParameter() {defaultValue=<rgba 0 1 1 1>, paramNames [RandomOffset,Param2,Param3,Param4]}.RandomOffset) {name=Dynamic_RandomOffset, nodeColor=<rgba 0 1 0.9063 1>}
  n11 = SkyAtmosphereLightDirection()
  n12 = ScalarParameter() {defaultValue=0.3, parameterName=Tiling2_UV}
  n13 = ScalarParameter() {defaultValue=-1, parameterName=PanSpeed2_U}
  n14 = ScalarParameter() {defaultValue=0.1, parameterName=PanSpeed2_V}
  n15 = TextureSample(UVs=Panner(Coordinate=Multiply(A=@n5, B=@n12) {constB=0.3}, Speed=AppendVector(A=@n13, B=@n14)) {speedX=-0.2}) {samplerType=SAMPLERTYPE_Normal, texture=/Game/Characters/Player/Male/Medium/Bodies/M_MED_Mastermind/Textures/T_M_MED_Mastermind_Twisted_N}
  n16 = ReflectionVectorWS(CustomWorldNormal=MaterialFunctionCall(BaseNormal=VertexNormalWS(), AdditionalNormal=NamedRerouteDeclaration(None=@n15.RGB) {name=PanningNormals, nodeColor=<rgba 0.6875 0 1 1>}) {materialFunction=&BlendAngleCorrectedNormals}.Result) {bNormalizeCustomWorldNormal=true}
  n17 = Power(Base=ConstantBiasScale(None=DotProduct(A=@n11, B=@n16)), Exp=ScalarParameter() {defaultValue=256, parameterName=MaskPower}) {constExponent=0.5}
  NamedRerouteDeclaration(None=LinearInterpolate(B=Multiply(A=@n17) {constB=2}, Alpha=ScalarParameter() {defaultValue=1, parameterName=UseAdadMagic})) {name=AdadMagic, nodeColor=<rgba 0 1 0.3672 1>}
  n18 = TextureSample(UVs=NamedRerouteUsage()) {samplerType=SAMPLERTYPE_Normal, texture=&/T_FX_MissusMind_Ink_N}
  n19 = AppendVector(A=Multiply(A=ComponentMask(None=@n18.RGB) {g=true, r=true}, B=ScalarParameter() {defaultValue=1.5, parameterName=NormalIntensity}), B=@n18.B)
  n20 = TextureSample(UVs=Multiply(A=ComponentMask(None=@n19) {g=true, r=true}, B=TextureCoordinate())) {texture=/Game/Packages/Fortress_SharedMaterials/Textures/T_Ravens_SpecularReflection}
  n21 = ScalarParameter() {parameterName=UseEmissive}
  NamedRerouteDeclaration(None=MaterialFunctionCall(Emissive=LinearInterpolate(B=@n20.G, Alpha=@n21)) {materialFunction=&MF_FX_EmissiveAdjustments}.Result) {name=Emissive, nodeColor=<rgba 0 0 0 0>}
  NamedRerouteDeclaration(None=@n10.A) {name=ParticleAlpha, nodeColor=<rgba 0 0.5547 1 1>}
  NamedRerouteDeclaration(None=MaterialFunctionCall(BaseNormal=@n19, AdditionalNormal=NamedRerouteUsage()) {materialFunction=&BlendAngleCorrectedNormals}.Result) {name=Normal, nodeColor=<rgba 0 0 0 1>}
  NamedRerouteUsage()
  NamedRerouteUsage()
  NamedRerouteUsage()
  NamedRerouteUsage()
  NamedRerouteUsage()
  NamedRerouteUsage()
 ]
}
