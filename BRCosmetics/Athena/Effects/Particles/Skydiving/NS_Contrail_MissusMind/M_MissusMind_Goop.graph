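# Masked, two-sided Niagara mesh-particle material for the dark, animated goop in the MissusMind skydiving contrail.
# World position offset combines time-driven Z-axis twist, normal-directed ripples from panning gradient textures and a sine bounce, using vertex-color masks and Niagara dynamic parameters.
# Rotation-corrected, blended normals accompany the deformation.
# A panning ink texture stepped against a UV gradient cuts opacity, while particle color, Fresnel and emissive adjustments supply shading inputs.
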Material /BRCosmetics/Athena/Effects/Particles/Skydiving/NS_Contrail_MissusMind/M_MissusMind_Goop
graph {
 properties {
  bAutomaticallySetUsageInEditor=false
  bUsedWithNiagaraMeshParticles=true
  blendMode=BLEND_Masked
  translucencyLightingMode=TLM_VolumetricPerVertexDirectional
  twoSided=true
  usedShadingModels=Default Lit
 }
 nodes [
  n1 = NamedRerouteUsage()
  n2 = VertexColor()
  n3 = Multiply(A=Time(), B=LinearInterpolate(A=NamedRerouteUsage(), B=Multiply(A=@n1, B=@n2.B), Alpha=@n2.G))
  n4 = Add(A=@n3, B=Multiply(A=Sine(None=@n3)) {constB=0.02}) {constB=0}
  n5 = Constant3Vector()
  n6 = VertexNormalWS()
  n7 = ComponentMask(None=TextureCoordinate()) {g=true}
  n8 = ScalarParameter() {defaultValue=2, parameterName=RipplesFalloff}
  n9 = Multiply(A=Saturate(None=Power(Base=@n7, Exp=@n8)), B=Multiply(A=NamedRerouteUsage(), B=NamedRerouteUsage()))
  n10 = Add(A=Multiply(A=Multiply(A=@n9, B=@n6), B=NamedRerouteUsage()), B=NamedRerouteUsage())
  MaterialFunctionCall(
   BaseNormal=MaterialFunctionCall(
    Rotation Axis=Constant3Vector() {constant=<rgba 0 0 1 0>}
    Rotation Angle=@n4
    World Space Vertex Normals=VertexNormalWS()
    Normals Pivot=@n5
   ) {materialFunction=&FixRotateAboutAxisNormals}.New Tangent Space Vertex Normal
   AdditionalNormal=MaterialFunctionCall(Rotation Axis=@n6, Rotation Angle=@n10, Normals Pivot=@n5) {materialFunction=&FixRotateAboutAxisNormals}.New Tangent Space Vertex Normal
  ) {materialFunction=&BlendAngleCorrectedNormals}
  Multiply(A=LinearInterpolate(
    A=VectorParameter() {parameterName=Color_Dark, defaultValue=<rgba 0.0144 0.0144 0.0203 1>}.RGB
    B=VectorParameter() {parameterName=Color_Light, defaultValue=<rgba 0.0439 0.0455 0.074 1>}.RGB
    Alpha=NamedRerouteUsage()
   ) {constAlpha=0})
  Multiply(A=NamedRerouteUsage(), B=ScalarParameter() {defaultValue=0.5, parameterName=EmissiveStrength})
  NamedRerouteDeclaration(None=ScalarParameter() {parameterName=Roughness}) {name=Roughness, nodeColor=<rgba 0 0 0 1>}
  NamedRerouteDeclaration(None=MaterialFunctionCall(Emissive=NamedRerouteUsage()) {materialFunction=&MF_FX_EmissiveAdjustments}.Result) {name=Emissive, nodeColor=<rgba 0 0 0 1>}
  NamedRerouteDeclaration(None=ParticleColor().RGB) {name=BaseColor, nodeColor=<rgba 0 0 0 1>}
  n11 = TextureCoordinate() {vTiling=0.5}
  n12 = ComponentMask(None=@n11) {r=true}
  n13 = Power(Base=ComponentMask(None=@n11) {g=true}) {constExponent=1}
  n14 = TextureSample(UVs=Panner(Coordinate=AppendVector(A=@n12, B=@n13)) {speedY=0.75}) {texture=/BRCosmetics/Effects/Fort_Effects/Textures/Grad/gradient_stasis}
  n15 = TextureCoordinate() {vTiling=0.5}
  n16 = ComponentMask(None=@n15) {r=true}
  n17 = Power(Base=ComponentMask(None=@n15) {g=true})
  n18 = TextureSample(UVs=Panner(Coordinate=AppendVector(A=@n16, B=@n17)) {speedY=2}) {texture=&gradient_stasis}
  n19 = NamedRerouteUsage()
  n20 = VertexColor()
  NamedRerouteDeclaration(None=Multiply(A=LinearInterpolate(A=@n14.G, B=@n18.G, Alpha=@n19), B=OneMinus(None=@n20.G))) {name=Output_RipplesMask, nodeColor=<rgba 0.9453 1 0 1>}
  n21 = DynamicParameter() {defaultValue=<rgba 0 10 1 0>, paramNames [RippleSpeed, RipplesStrength, TwistSpeed, BitsSpin]}
  NamedRerouteDeclaration(None=@n21.RippleSpeed) {name=Dynamic_RippleSpeed, nodeColor=<rgba 0 0.625 1 1>}
  NamedRerouteDeclaration(None=@n21.RipplesStrength) {name=Dynamic_RipplesStrength, nodeColor=<rgba 0.9453 1 0 1>}
  n22 = Fresnel(ExponentIn=ScalarParameter() {defaultValue=1, parameterName=Fresnel_Exponent})
  n23 = ScalarParameter() {defaultValue=1, parameterName="Contrast (S)"}
  NamedRerouteDeclaration(None=Saturate(None=MaterialFunctionCall(In=@n22, Contrast=@n23) {materialFunction=&CheapContrast}.Result)) {name=FresnelMask, nodeColor=<rgba 0.125 1 0 1>}
  n24 = TextureCoordinate()
  n25 = Power(Base=OneMinus(None=ComponentMask(None=@n24) {g=true}), Exp=ScalarParameter() {defaultValue=2, parameterName=OpacityMaskPower})
  n26 = Panner() {speedX=1, speedY=2}
  NamedRerouteDeclaration(None=Step(
    Y=Saturate(None=@n25)
    X=TextureSample(UVs=@n26) {samplerType=SAMPLERTYPE_Masks, texture=&/T_FX_MissusMind_Ink_M}.G
   )) {name=OpacityMask, nodeColor=<rgba 0 0 0 1>}
  NamedRerouteDeclaration(None=ScalarParameter() {defaultValue=0.2, parameterName=Specular}) {name=Specular, nodeColor=<rgba 0 0 0 1>}
  NamedRerouteDeclaration(None=Add(A=NamedRerouteUsage(), B=NamedRerouteUsage())) {name=WPO, nodeColor=<rgba 0 0 0 1>}
  NamedRerouteDeclaration(None=@n21.TwistSpeed) {name=Dynamic_TwistSpeed, nodeColor=<rgba 1 0.3516 0 1>}
  NamedRerouteDeclaration(None=@n21.BitsSpin) {name=Dynamic_BitsSpin, nodeColor=<rgba 0.3828 1 0 1>}
  n27 = Constant3Vector() {constant=<rgba 0 0 10 0>}
  n28 = Constant3Vector()
  n29 = Time()
  n30 = VertexColor()
  n31 = ConstantBiasScale(None=Sine(None=Add(A=@n29, B=@n30.B)) {period=2})
  NamedRerouteDeclaration(None=Multiply(A=LinearInterpolate(A=@n27, B=@n28, Alpha=@n31), B=VertexColor().G)) {name=Output_Bounce, nodeColor=<rgba 0 0.9531 1 1>}
  n32 = OneMinus(None=ComponentMask(None=TextureCoordinate()) {g=true})
  n33 = ScalarParameter() {defaultValue=1, parameterName=GradientMask_Exp}
  NamedRerouteDeclaration(None=Saturate(None=Power(Base=@n32, Exp=@n33) {constExponent=5})) {name=GradientMask, nodeColor=<rgba 0.5234 0 1 1>}
  NamedRerouteDeclaration(None=MaterialFunctionCall!(Rotation Amount=@n4, PivotPoint, WorldPosition=WorldPosition().XYZ) {materialFunction=/Engine/Functions/Engine_MaterialFunctions02/WorldPositionOffset/RotateAboutWorldAxis_cheap}.Z-Axis) {name=WPO_Twist, nodeColor=<rgba 0 1 0.2031 1>}
  NamedRerouteDeclaration(None=@n10) {name=WPO_Bounce, nodeColor=<rgba 1 0 0.1172 1>}
  NamedRerouteUsage()
  NamedRerouteUsage()
  NamedRerouteUsage()
  NamedRerouteUsage()
  NamedRerouteUsage()
  NamedRerouteUsage()
  NamedRerouteUsage()
 ]
}
