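# Niagara skydiving contrail for the MissusMind cosmetic, with dark goo blobs around the character, attached ribbons and speed lines.
# OrbitBlobs follows the skeleton with smoothed offsets, wobble and sine motion, emitting location events, presumably feeding its drip and trail emitters.
# Contrail setup's Dive Value blends blob layout, stretch and material, plus trail and streak speed.
# Front-end preview shifts blobs and stops speed lines; hand emitters are disabled.
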NiagaraSystem /BRCosmetics/Athena/Effects/Particles/Skydiving/NS_Contrail_MissusMind/NS_MissusMind_Contrail
dependencies {
 usedDataInterfaces [
  &NiagaraDataInterfaceArrayFloat
  &NiagaraDataInterfaceCurve
  &NiagaraDataInterfaceMeshRendererInfo
  &NiagaraDataInterfaceVectorCurve
  &NiagaraDataInterfaceArrayFloat3
 ]
 usedDynamicInputs [
  &MakeFloatFromVector
  &SetBoolByFloatComparison
  &Divide_Float
  &Abs_Float
 ]
 usedModules [
  &self:SystemSpawnScript.SystemScriptSource.SystemScriptGraph.NiagaraNodeAssignment_0.SetVariables_#1
  &SystemState
  &self:SystemSpawnScript.SystemScriptSource.SystemScriptGraph.NiagaraNodeAssignment_5.SetVariables_#2
  /Game/Effects/Niagara/Modules/Contrails/FN_Contrail_Setup
  &self:SystemSpawnScript.SystemScriptSource.SystemScriptGraph.NiagaraNodeAssignment_4.SetVariables_#3
  &self:SystemSpawnScript.SystemScriptSource.SystemScriptGraph.NiagaraNodeAssignment_1.SetVariables_#4
  &Timeline
  &self:SystemSpawnScript.SystemScriptSource.SystemScriptGraph.NiagaraNodeAssignment_2.SetVariables_#5
  &self:SystemSpawnScript.SystemScriptSource.SystemScriptGraph.NiagaraNodeAssignment_3.SetVariables_#6
  &DoOnce
  /Game/Effects/Niagara/Modules/Skydiving/FN_DivingAttributes/v6/FN_DivingAttributes
 ]
}
emitters {
 FN_AttachedRibbon {
  emitterUpdateScript [
   &EmitterState()
   &FN_SpawnRibbon_v2(Number of Ribbons=2, Spawn Mode=::NewEnumerator0, SpawnRate=30)
   &FN_Integrator() {enabled=false}
  ]
  particleSpawnScript [
   /Game/Effects/Niagara/Modules/Ribbons/FN_InitializeRibbon_v4(Enable Skeletal Position=true, SkeletalMesh:NiagaraDataInterfaceSkeletalMesh=$User.SkeletalMesh)
   &FN_FlipRibbonV(Seed:NiagaraInt32=&ReturnParticleIDAttributes(Particle ID Attribute=::NewEnumerator2, Niagara ID=$Particles.RibbonID)) {enabled=false}
   &AddVelocity(Velocity=<0 0 250>)
   &DynamicMaterialParameters(
    Index 0 Param 1=&SelectFloatFromArray(
     Direct Array Index=&ReturnParticleIDAttributes(Particle ID Attribute=::NewEnumerator1, Niagara ID=$Particles.RibbonID)
     Float Selection Array=DataInterface() {propertyValues {FloatData [0,0.5]}}
    )
    Module.Index0Param0Write=true
    Module.Index0Param1Write=true
    Module.Index0Param2Write=true
    Module.Index0Param3Write=true
   )
  ]
  particleUpdateScript [
   &SolveForcesAndVelocity(Clamp Velocity=false, Limit Acceleration=false)
   /Game/Effects/Niagara/Modules/Ribbons/FN_UpdateRibbon_v4(Enable Skeletal Position=true, SkeletalMesh:NiagaraDataInterfaceSkeletalMesh=$User.SkeletalMesh)
   &FN_RibbonMaterial() {enabled=false}
   &ScaleRibbonWidth(Ribbon Width Scale=&Lerp_Float(
     A=1
     B=4
     Alpha=&FloatFromCurve(
      FloatCurve=DataInterface() {propertyValues {Curve {keys [
          {interpMode=RCIM_Cubic, tangentMode=RCTM_User, time=0, value=0}
          {arriveTangent=1.6527, interpMode=RCIM_Cubic, leaveTangent=1.6527, tangentMode=RCTM_User, time=1, value=1}
         ]}}}
      Scale Curve=2
     )
    ))
   &ScaleColor(
    Scale Alpha=&Lerp_Float(
     A=0
     B=1
     Alpha=&FloatFromCurve(FloatCurve=DataInterface() {propertyValues {Curve {keys [
          {arriveTangent=-1.219, interpMode=RCIM_Cubic, leaveTangent=-1.219, tangentMode=RCTM_User, time=0, value=1}
          {arriveTangent=-0.0448, interpMode=RCIM_Cubic, leaveTangent=-0.0448, tangentMode=RCTM_User, time=1, value=0}
         ]}}})
    )
    ScaleA=true
    ScaleRGB=true
   )
  ]
  rendererClasses [&NiagaraRibbonRendererProperties]
  renderers [{rendererClass=&NiagaraRibbonRendererProperties}]
 }
 Hands_BlobDrips {
  bEnabled=false
  emitterUpdateScript [
   &EmitterState(Life Cycle Mode=::NewEnumerator1, Loop Duration=1, UseLoopDelay=false)
   &SpawnRate(SpawnRate=50, Use Spawn Probability=false)
  ]
  particleSpawnScript [
   &InitializeParticle(
    Color=<rgba 0.01 0.01 0.01 1.5>
    Lifetime=&UniformAOrBFloat(A=0, B=1.5, Distribution Weight=0.15, Evaluation Type=::NewEnumerator0, Recalculate Random Each Loop=true, Override Seed=false)
    Mesh Scale Mode=::NewEnumerator1
    Mesh Uniform Scale Max=1
    Mesh Uniform Scale Min=0.5
    UsePositionOffset=false
    Write Material Random=true
   )
   &InitialMeshOrientation(Mesh Orientation Mode=::NewEnumerator2, Use Rotation Vector=false)
   &SkeletalMeshLocation(
    Bone / Socket Index=&RandomRangeInt(Minimum=0, Maximum=1)
    MaskSpawnGroup=false
    Mesh Sampling Type=::NewEnumerator5
    Sampled Position Offset=<10 0 0>
    Skeletal Mesh=$User.SkeletalMesh
    Socket Sampling Mode=::NewEnumerator5
   )
   &AddVelocity(Rotation Coordinate Space=::World, Velocity=&RandomRangeVector(Minimum=<0 0 250>, Maximum=<0 0 400>))
   &DynamicMaterialParameters(Module.Index0Param0Write=true, Module.Index0Param1Write=true, Module.Index0Param2Write=true, Module.Index0Param3Write=true)
  ]
  particleUpdateScript [
   &ParticleState()
   &ScaleMeshSize()
   &SolveForcesAndVelocity(Clamp Velocity=false, Limit Acceleration=false)
  ]
  rendererClasses [&NiagaraMeshRendererProperties]
  renderers [{rendererClass=&NiagaraMeshRendererProperties, rendererIndex=0}]
 }
 Hands_Blobs {
  bEnabled=false
  emitterUpdateScript [
   &EmitterState()
   &SpawnRate(SpawnRate=45, Use Spawn Probability=false)
  ]
  particleSpawnScript [
   &InitializeParticle(
    Color=<rgba 0.01 0.01 0.01 1.5>
    Lifetime=0.65
    Mesh Scale Mode=::NewEnumerator1
    Mesh Uniform Scale Max=5
    Mesh Uniform Scale Min=4
    UsePositionOffset=false
    Write Material Random=true
   )
   &InitialMeshOrientation(Mesh Orientation Mode=::NewEnumerator2, Use Rotation Vector=false)
  ]
  particleUpdateScript [
   &ParticleState()
   &DynamicMaterialParameters(
    Index 0 Param 3=&RandomRangeFloat(Minimum=0.1, Maximum=-0.1)
    Module.Index0Param0Write=false
    Module.Index0Param1Write=false
    Module.Index0Param2Write=true
    Module.Index0Param3Write=false
   )
   &SkeletalMeshLocation(
    Bone / Socket Index=&RandomRangeInt(Minimum=0, Maximum=1)
    MaskSpawnGroup=false
    Mesh Sampling Type=::NewEnumerator5
    Sampled Position Offset=&RandomRangeVector(Minimum=<12 5 -5>, Maximum=<8 -5 5>)
    Skeletal Mesh=$User.SkeletalMesh
    Socket Sampling Mode=::NewEnumerator5
   )
   &ScaleMeshSize(Scale Factor=&VectorFromFloat(Value=&FloatFromCurve(FloatCurve=DataInterface() {propertyValues {Curve {keys [
          {arriveTangent=2.7055, interpMode=RCIM_Cubic, leaveTangent=2.7055, tangentMode=RCTM_User, time=0, value=0.25}
          {interpMode=RCIM_Cubic, tangentMode=RCTM_User, time=0.2, value=0.95}
          {arriveTangent=-2.1694, tangentMode=RCTM_User, time=1, value=0.25}
         ]}}})))
   &AccelerationForce(Acceleration=<0 0 1600>)
   &SolveForcesAndVelocity(Clamp Velocity=false, Limit Acceleration=false)
  ]
  rendererClasses [&NiagaraMeshRendererProperties]
  renderers [{rendererClass=&NiagaraMeshRendererProperties, rendererIndex=0}]
 }
 Hands_Trails {
  bEnabled=false
  emitterUpdateScript [
   &EmitterState()
   &SpawnBurst_Instantaneous(Spawn Count=20, Use Spawn Probability=false, UseLoopCountLimit=true)
  ]
  particleSpawnScript [
   &PartitionParticles(
    Normalized Index In Partition Includes One=true
    Particles Per Partition=10
    Partition Mode:ENiagara_PartitionMode=::NewEnumerator1
    Write Partition Index to Ribbon Link Order:ENiagara_AttributeSamplingApplyOutput=::NewEnumerator0
    Write Partition to Ribbon ID=::NewEnumerator0
   )
   &InitializeParticle(
    Color Mode=::NewEnumerator0
    Lifetime=0.2
    Ribbon Width=3
    Ribbon Width Mode=::NewEnumerator1
    Sprite Size Mode=::NewEnumerator3
    UsePositionOffset=false
    Write Material Random=true
   )
  ]
  particleUpdateScript [
   &ParticleState(Kill Particles When Lifetime Has Elapsed=false)
   &SkeletalMeshLocation(
    Bone / Socket Index=$Particles.PartitionParticles.Partition
    MaskSpawnGroup=false
    Mesh Sampling Type=::NewEnumerator5
    Sampled Position Offset=<10 0 0>
    Skeletal Mesh=$User.SkeletalMesh
    Socket Sampling Mode=::NewEnumerator5
   )
   /SetVariables(Particles.NewAge:NiagaraFloat=&Multiply_Float(A=$Particles.PartitionParticles.Normalized Index In Partition, B=$Particles.Initial.Lifetime)) {isSetParameters=true}
   /SetVariables(Particles.Velocity=&Add_Vector(
     A=&Multiply_VectorByFloat(
      Vector=&Lerp_Vector(A=$System.FN_Contrail_Setup.Direction Velocity Raw, B=$System.FN_Contrail_Setup.Direction Velocity, Alpha=$Particles.NewAge)
      Float=$Particles.Initial.Lifetime
     )
     B=<0 0 0>
    )) {isSetParameters=true}
   /SetVariables(Particles.Position=&AddVectorToPosition(
     Position=$Particles.Position
     Vector=&Multiply_VectorByFloat(Vector=$Particles.Velocity, Float=$Particles.PartitionParticles.Normalized Index In Partition)
    )) {isSetParameters=true, module=SetVariables_#7}
   &ScaleColor(
    Scale Alpha=&FloatFromCurve(
     FloatCurve=DataInterface() {propertyValues {Curve {keys [
         {interpMode=RCIM_Cubic, tangentMode=RCTM_User, time=0, value=1}
         {arriveTangent=-2.9745, interpMode=RCIM_Cubic, leaveTangent=-2.9745, tangentMode=RCTM_User, time=1, value=0}
        ]}}}
     CurveIndex=$Particles.PartitionParticles.Normalized Index In Partition
    )
    Scale RGB=<0.1 0.1 0.1>
    ScaleA=true
    ScaleRGB=true
   )
   &ScaleRibbonWidth(Ribbon Width Scale=&Lerp_Float(
     A=1
     B=3
     Alpha=&FloatFromCurve(
      FloatCurve=DataInterface() {propertyValues {Curve {keys [
          {arriveTangent=0.41, interpMode=RCIM_Cubic, leaveTangent=0.41, tangentMode=RCTM_User, time=0, value=0}
          {interpMode=RCIM_Cubic, tangentMode=RCTM_User, time=1, value=1}
         ]}}}
      CurveIndex=$Particles.PartitionParticles.Normalized Index In Partition
     )
    ))
  ]
  rendererClasses [&NiagaraRibbonRendererProperties]
  renderers [{rendererClass=&NiagaraRibbonRendererProperties}]
 }
 OrbitBlobs {
  emitterUpdateScript [
   &EmitterState(
    Inactive Response=::NewEnumerator1
    Life Cycle Mode=::NewEnumerator1
    Loop Behavior=::NewEnumerator1
    Loop Duration Mode=::NewEnumerator1
    UseLoopDelay=false
   )
   &SpawnBurst_Instantaneous(
    Spawn Count=&MakeIntFromFloat(Float=&FN_FrontEndFloatModifier(Gameplay=4, Front End=4))
    Use Spawn Probability=false
    UseLoopCountLimit=true
   )
   &InterpolateOverTime(Rate Of Change=50, Target Vector Value:Vector3f=$Engine.Owner.Velocity, Variable Type=::NewEnumerator2)
  ]
  particleSpawnScript [
   &FN_SetSpawnGroup() {module=FN_SetSpawnGroup001}
   /SetVariables(Particles.OrbitSpeed=1) {isSetParameters=true}
   /SetVariables(Particles.SpawnGroup2:NiagaraInt32=$Particles.SpawnGroup) {isSetParameters=true}
   &FN_SetSpawnGroup(Number of Spawn Groups=2)
   &InitializeParticle(
    Color=<rgba 0.01 0.01 0.01 1.5>
    Lifetime=2
    Mesh Scale=&VectorFromFloat(Value=&FN_FrontEndFloatModifier(
      Gameplay=&SelectFloatFromArray(Direct Array Index=$Particles.SpawnGroup2, Float Selection Array=DataInterface() {propertyValues {FloatData [1.6,1.6,1,1]}})
      Front End=&SelectFloatFromArray(Direct Array Index=$Particles.SpawnGroup2, Float Selection Array=DataInterface() {propertyValues {FloatData [1.6,1.6,1.1,1.1]}})
     ))
    Mesh Scale Mode=::NewEnumerator4
    UsePositionOffset=true
    Write Material Random=true
   )
  ]
  particleUpdateScript [
   &ParticleState(Kill Particles When Lifetime Has Elapsed=false, Let Infinitely Lived Particles Die When Emitter Deactivates=true)
   &DynamicMaterialParameters(
    Index 0 Param 1=&Lerp_Float(
     A=0
     B=1
     Alpha=&FloatFromCurve(
      FloatCurve=DataInterface() {propertyValues {Curve {keys [
          {arriveTangent=5.87, interpMode=RCIM_Cubic, leaveTangent=5.87, tangentMode=RCTM_User, time=0, value=0}
          {interpMode=RCIM_Cubic, tangentMode=RCTM_User, time=0.5, value=1}
         ]}}}
      CurveIndex=$System.FN_Contrail_Setup.Dive Value
     )
    )
    Index 0 Param 2=&Lerp_Float(
     A=5
     B=10
     Alpha=&FloatFromCurve(
      FloatCurve=DataInterface() {propertyValues {Curve {keys [
          {arriveTangent=5.87, interpMode=RCIM_Cubic, leaveTangent=5.87, tangentMode=RCTM_User, time=0, value=0}
          {interpMode=RCIM_Cubic, tangentMode=RCTM_User, time=0.5, value=1}
         ]}}}
      CurveIndex=$System.FN_Contrail_Setup.Dive Value
     )
    )
    Index 0 Param 3=&RandomRangeFloat(Minimum=0.35, Maximum=0.5)
    Index 0 Param 4=&RandomRangeFloat(Minimum=0.7, Maximum=0.9)
    Module.Index0Param0Write=true
    Module.Index0Param1Write=true
    Module.Index0Param2Write=true
    Module.Index0Param3Write=true
   )
   &SkeletalMeshLocation(
    Bone / Socket Index=2
    Bone Sampling Mode=::NewEnumerator4
    MaskSpawnGroup=false
    Orientation Sampling=::NewEnumerator1
    Sampled Orientation Coordinate Space=::World
    Skeletal Mesh=$System.Skeletal Mesh
    Use Sampled Mesh Attribute as Ribbon ID=false
   )
   /SetVariables(Particles.AmbientMovement:Vector3f=&Multiply_VectorByFloat(
     Vector=&MakeVector(
      X=&Waveform(
       Global Amplitude Scale=0.5
       "Waveform [1]"=::NewEnumerator12
       "[1] Input Position"=&RandomRangeFloat(Minimum=0, Maximum=1000)
       "[1] Frequency"=1
       "[1] Amplitude Min/Max"=<-50 50>
      )
      Y=&Waveform(
       "Waveform [1]"=::NewEnumerator12
       "[1] Input Position"=&RandomRangeFloat(Minimum=1200, Maximum=1800)
       "[1] Frequency"=1
       "[1] Amplitude Min/Max"=<-50 50>
      )
      Z=&Waveform(
       Global Amplitude Scale=1.5
       "Waveform [1]"=::NewEnumerator12
       "[1] Input Position"=&RandomRangeFloat(Minimum=2200, Maximum=2800)
       "[1] Frequency"=1
       "[1] Amplitude Min/Max"=<-50 50>
      )
     )
     Float=0.3333
    )) {isSetParameters=true}
   /SetVariables(Particles.HF jiggle:Vector3f=&Multiply_VectorByFloat(
     Vector=&MakeVector(
      X=&Waveform(
       Global Input Position=&RandomRangeFloat(Minimum=0, Maximum=1)
       "Waveform [1]"=::NewEnumerator12
       "[1] Input Position"=&RandomRangeFloat(Minimum=800, Maximum=1600)
       "[1] Frequency"=8
       "[1] Amplitude Min/Max"=<-10 10>
       "[1] Scale"=0.333
      )
      Y=&Waveform(
       Global Input Position=&RandomRangeFloat(Minimum=0, Maximum=1)
       "Waveform [1]"=::NewEnumerator12
       "[1] Input Position"=&RandomRangeFloat(Minimum=1400, Maximum=2000)
       "[1] Frequency"=8
       "[1] Amplitude Min/Max"=<-10 10>
       "[1] Scale"=0.333
      )
      Z=&Waveform(
       Global Input Position=&RandomRangeFloat(Minimum=0, Maximum=1)
       "Waveform [1]"=::NewEnumerator12
       "[1] Input Position"=&RandomRangeFloat(Minimum=0, Maximum=1000)
       "[1] Frequency"=8
       "[1] Amplitude Min/Max"=<-10 10>
       "[1] Scale"=0.333
      )
     )
     Float=&Lerp_Float(A=0, B=0.8, Alpha=$System.InterpolateOverTime.InterpolateOverTime_Float.Moving Average)
    )) {isSetParameters=true}
   &self:ScratchModule_02(InterpolatedQuat:Quat4f=$Particles.SkeletalMeshLocation.SampledRotation, WorldToLocalMatrix:NiagaraMatrix=$Engine.Owner.SystemWorldToLocal) {module=ScratchModule_02}
   &InterpolateOverTime(Rate Of Change=6, Target Quat Value:Quat4f=$Particles.SkeletalMeshLocation.SampledRotation, Variable Type=::NewEnumerator5) {module=InterpolateOverTime001}
   &self:ScratchModule_01(
    InterpolatedQuat:Quat4f=$Particles.InterpolateOverTime001.InterpolateOverTime_Quaternion.Moving Average
    WorldToLocalMatrix:NiagaraMatrix=$Engine.Owner.SystemWorldToLocal
   ) {module=ScratchModule_01}
   /SetVariables(Particles.RotAmount:NiagaraFloat=&self:ScratchDynamicInput_02(
     InitialMeshVector:Vector3f=&ReturnMeshOrientationAxis(Mesh Orientation Axis=::ZAxis, Mesh Orientation Quat=$Particles.InitialMeshOrientation)
     MeshVector:Vector3f=&ReturnMeshOrientationAxis(Mesh Orientation Axis=::ZAxis, Mesh Orientation Quat=$Particles.MeshOrientation)
    )) {isSetParameters=true}
   /SetVariables(Particles.ShapeVectorToWorld:Vector3f=&TransformVector(Vector=$Particles.ShapeLocation002.ShapeVector, Destination Space=::World)) {isSetParameters=true}
   &InterpolateOverTime(
    Rate Of Change:NiagaraFloat=&SelectFloatFromArray(Direct Array Index=$Particles.SpawnGroup2, Float Selection Array=DataInterface() {propertyValues {FloatData [6,10,3,6,10,4]}})
    Target Vector Value:Vector3f=$Particles.ShapeVectorToWorld
    Variable Type=::NewEnumerator2
   ) {module=InterpolateOverTime002}
   /SetVariables(Particles.MeshOrientation=&OrientMeshToVector(Look At Direction=$Particles.ShapeLocation002.ShapeTangent, Reference Axis Vector=<1 0 0>)) {isSetParameters=true}
   /SetVariables(Particles.MeshOrientation=&MultiplyQuaternion(
     Quaternion A=$Particles.MeshOrientation
     Quaternion B=&MakeQuatFromAxisAngle(
      Axis=<0 1 0>
      AngleInDegrees=&Lerp_Float(A=0, B=0, Alpha=$System.InterpolateOverTime.InterpolateOverTime_Float.Moving Average)
     )
    )) {isSetParameters=true}
   /SetVariables(Particles.CatchupVector:Vector3f=&Multiply_VectorByFloat(
     Vector=&Multiply_VectorByFloat(
      Vector=&NormalizeVector(Vector To Normalize=$System.SkydiveVelocity, Safe Normalize=true, Safe Fallback Threshold=0, Safe Fallback Vector=<0 0 0.001>)
      Float=&FN_FrontEndFloatModifier(Gameplay=0, Front End=0)
     )
     Float=&Lerp_Float(
      A=&FloatFromCurve(
       FloatCurve=DataInterface() {propertyValues {Curve {keys [
           {interpMode=RCIM_Cubic, tangentMode=RCTM_User, time=0, value=0}
           {interpMode=RCIM_Cubic, tangentMode=RCTM_User, time=0.35, value=-0.35}
           {tangentMode=RCTM_User, time=1, value=0}
          ]}}}
       CurveIndex=$System.Timeline001.NormalizedTime
      )
      B=&FloatFromCurve(
       FloatCurve=DataInterface() {propertyValues {Curve {keys [
           {arriveTangent=3.3827, arriveTangentWeight=1.1106, interpMode=RCIM_Cubic, leaveTangentWeight=1.1106, time=0, value=0}
           {arriveTangentWeight=0.5241, interpMode=RCIM_Cubic, leaveTangentWeight=0.1667, time=0.5, value=0.65}
           <at 1 0 RCIM_Cubic>
          ]}}}
       CurveIndex=$System.Timeline001.NormalizedTime
      )
      Alpha=$Particles.InterpolateOverTime004.InterpolateOverTime_Float.Moving Average
     )
    )) {isSetParameters=true}
   /SetVariables(Particles.InitialOffset:Vector3f=&Lerp_Vector(
     A=&FN_FrontEndVector3Modifier(Gameplay=<0 0 500>, Front End=<0 0 300>)
     B=&Lerp_Vector(A=<0 0 0>, B=<0 0 0>, Alpha=$System.InterpolateOverTime.InterpolateOverTime_Float.Moving Average)
     Alpha=&FloatFromCurve(
      FloatCurve=DataInterface() {propertyValues {Curve {keys [
          {arriveTangent=1.5577, interpMode=RCIM_Cubic, leaveTangent=1.5577, tangentMode=RCTM_User, time=0, value=0}
          {arriveTangent=-0.0191, interpMode=RCIM_Cubic, leaveTangent=-0.0191, tangentMode=RCTM_User, time=1, value=1}
         ]}}}
      CurveIndex=&Multiply_Float(A=$Engine.System.Age, B=&RandomRangeFloat(Minimum=0.7, Maximum=1))
     )
    )) {isSetParameters=true}
   /SetVariables(Particles.MainOffset:Vector3f=&Add_Vector(
     A=<0 0 0>
     B=&Add_Vector(
      A=$Particles.InitialOffset
      B=&Add_Vector(
       A=$Particles.AmbientMovement
       B=&Add_Vector(A=$Particles.CatchupVector, B=&Add_Vector(A=<0 0 0>, B=$Particles.HF jiggle))
      )
     )
    )) {isSetParameters=true}
   /SetVariables(Particles.Position=&AddVectorToPosition(
     Position=$Particles.SkeletalMeshLocation.SampledPosition
     Vector=&Add_Vector(
      A=&TransformVector(Vector=$Particles.InterpolateOverTime002.InterpolateOverTime_V3_.Moving Average, Source Space=::World, Destination Space=::Local)
      B=&Add_Vector(
       A=$Particles.MainOffset
       B=&Multiply_Vector(
        A=&MakeVector(
         X=&Sine(
          Normalized Angle=$Emitter.Age
          Period=&Add_Float(A=15, B=&MakeFloatFromInt(Int=&ReturnParticleIDAttributes()))
          Scale=2
          Bias=0
         )
         Y=&Cosine(
          Normalized Angle=$Emitter.Age
          Period=&Add_Float(A=15, B=&MakeFloatFromInt(Int=&ReturnParticleIDAttributes()))
          Scale=1.8
          Bias=0
         )
         Z=1
        )
        B=&Lerp_Vector(
         A=&AddVectorToPosition(
          Position=&SelectVectorFromArray(Direct Array Index=&ReturnParticleIDAttributes(), Vector Selection Array=$System.GooLocations)
          Vector=&MakeCustomVectorFromBool(Bool=$User.bIsFrontEndPreview, True Vector=<0 0 50>, False Vector=<0 0 0>)
         )
         B=&AddVectorToPosition(
          Position=&MakeCustomVectorFromBool(Bool=$User.bIsFrontEndPreview, True Vector=<0 0 170>, False Vector=<0 0 0>)
          Vector=&SelectVectorFromArray(Direct Array Index=&ReturnParticleIDAttributes(), Vector Selection Array=$System.GooLocations_Diving)
         )
         Alpha=&FloatFromCurve(
          FloatCurve=DataInterface() {propertyValues {Curve {keys [
              {arriveTangent=1.5326, interpMode=RCIM_Cubic, leaveTangent=1.5326, tangentMode=RCTM_User, time=0, value=0}
              {interpMode=RCIM_Cubic, tangentMode=RCTM_User, time=1, value=1}
             ]}}}
          CurveIndex=$System.FN_Contrail_Setup.Dive Value
         )
        )
       )
      )
     )
    )) {isSetParameters=true}
   /SetVariables(Particles.Position=&AddVectorToPosition(Position=$Particles.Position, Vector=<0 0 0>)) {enabled=false, isSetParameters=true}
   /SetVariables(Particles.MeshOrientation=&MultiplyQuaternion(Quaternion A=$Particles.MeshOrientation, Quaternion B=$Particles.Wobble)) {isSetParameters=true}
   &SolveForcesAndVelocity(Clamp Velocity=false, Limit Acceleration=false)
   &ScaleMeshSize(Scale Factor=&Lerp_Vector(
     A=<1.1 1.1 1>
     B=<0.8 0.8 1.5>
     Alpha=&FloatFromCurve(
      FloatCurve=DataInterface() {propertyValues {Curve {keys [
          {arriveTangent=2.0291, interpMode=RCIM_Cubic, leaveTangent=2.0291, tangentMode=RCTM_User, time=0, value=0}
          {interpMode=RCIM_Cubic, tangentMode=RCTM_User, time=1, value=1}
         ]}}}
      CurveIndex=$System.FN_Contrail_Setup.Dive Value
     )
    ))
   &GenerateLocationEvent(Use Event Probability=false, UseEventDelay=false)
  ]
  rendererClasses [&NiagaraMeshRendererProperties]
  renderers [{rendererClass=&NiagaraMeshRendererProperties, rendererIndex=0}]
 }
 OrbitBlobs001 {
  emitterUpdateScript [
   &EmitterState(
    Inactive Response=::NewEnumerator1
    Life Cycle Mode=::NewEnumerator1
    Loop Behavior=::NewEnumerator1
    Loop Duration Mode=::NewEnumerator1
    UseLoopDelay=false
   )
   &SpawnBurst_Instantaneous(
    Spawn Count=&MakeIntFromFloat(Float=&FN_FrontEndFloatModifier(Gameplay=4, Front End=4))
    Use Spawn Probability=false
    UseLoopCountLimit=true
   )
   &InterpolateOverTime(Rate Of Change=50, Target Vector Value:Vector3f=$Engine.Owner.Velocity, Variable Type=::NewEnumerator2)
  ]
  particleSpawnScript [
   &FN_SetSpawnGroup(Number of Spawn Groups=2)
   &InitializeParticle(
    Color=<rgba 0.01 0.01 0.01 1.5>
    Lifetime=2
    Mesh Scale Mode=::NewEnumerator3
    Mesh Uniform Scale=1.5
    UsePositionOffset=true
    Write Material Random=true
   )
  ]
  particleUpdateScript [
   &ParticleState(Kill Particles When Lifetime Has Elapsed=false, Let Infinitely Lived Particles Die When Emitter Deactivates=true)
   &DynamicMaterialParameters(
    Index 0 Param 1=&Lerp_Float(
     A=0
     B=1
     Alpha=&FloatFromCurve(
      FloatCurve=DataInterface() {propertyValues {Curve {keys [
          {arriveTangent=5.87, interpMode=RCIM_Cubic, leaveTangent=5.87, tangentMode=RCTM_User, time=0, value=0}
          {interpMode=RCIM_Cubic, tangentMode=RCTM_User, time=0.5, value=1}
         ]}}}
      CurveIndex=$System.FN_Contrail_Setup.Dive Value
     )
    )
    Index 0 Param 2=&Lerp_Float(
     A=5
     B=10
     Alpha=&FloatFromCurve(
      FloatCurve=DataInterface() {propertyValues {Curve {keys [
          {arriveTangent=5.87, interpMode=RCIM_Cubic, leaveTangent=5.87, tangentMode=RCTM_User, time=0, value=0}
          {interpMode=RCIM_Cubic, tangentMode=RCTM_User, time=0.5, value=1}
         ]}}}
      CurveIndex=$System.FN_Contrail_Setup.Dive Value
     )
    )
    Index 0 Param 3=&RandomRangeFloat(Minimum=0.35, Maximum=0.5)
    Index 0 Param 4=&RandomRangeFloat(Minimum=0.7, Maximum=0.9)
    Module.Index0Param0Write=true
    Module.Index0Param1Write=true
    Module.Index0Param2Write=true
    Module.Index0Param3Write=true
   )
   &SkeletalMeshLocation(
    Bone / Socket Index=&Add_Integer(A=1, B=$Particles.SpawnGroup)
    Bone Sampling Mode=::NewEnumerator4
    MaskSpawnGroup=false
    Orientation Sampling=::NewEnumerator1
    Sampled Orientation Coordinate Space=::World
    Skeletal Mesh=$System.Skeletal Mesh
    Use Sampled Mesh Attribute as Ribbon ID=false
   )
   &self:ScratchModule_01(
    InterpolatedQuat:Quat4f=$Particles.InterpolateOverTime001.InterpolateOverTime_Quaternion.Moving Average
    WorldToLocalMatrix:NiagaraMatrix=$Engine.Owner.SystemWorldToLocal
   ) {module=ScratchModule_01}
   &ScaleMeshSize(Scale Factor=&Lerp_Vector(
     A=<1.1 1.1 1>
     B=<0.8 0.8 1.5>
     Alpha=&FloatFromCurve(
      FloatCurve=DataInterface() {propertyValues {Curve {keys [
          {arriveTangent=2.0291, interpMode=RCIM_Cubic, leaveTangent=2.0291, tangentMode=RCTM_User, time=0, value=0}
          {interpMode=RCIM_Cubic, tangentMode=RCTM_User, time=1, value=1}
         ]}}}
      CurveIndex=$System.FN_Contrail_Setup.Dive Value
     )
    ))
  ]
  rendererClasses [&NiagaraMeshRendererProperties]
  renderers [{rendererClass=&NiagaraMeshRendererProperties, rendererIndex=0}]
 }
 OrbitBlobs_Drips {
  emitterUpdateScript [&EmitterState()]
  particleSpawnScript [
   &InitializeParticle(
    Color=<rgba 0.01 0.01 0.01 1.5>
    Lifetime=&UniformAOrBFloat(A=0, B=1.5, Distribution Weight=0.1, Evaluation Type=::NewEnumerator0, Recalculate Random Each Loop=true, Override Seed=false)
    Mesh Scale Mode=::NewEnumerator1
    UsePositionOffset=false
    Write Material Random=true
   )
   &InitialMeshOrientation(Mesh Orientation Mode=::NewEnumerator2, Use Rotation Vector=false)
   &AddVelocity(Rotation Coordinate Space=::World, Velocity=&RandomRangeVector(Minimum=<0 0 250>, Maximum=<0 0 400>))
   &DynamicMaterialParameters(Module.Index0Param0Write=true, Module.Index0Param1Write=true, Module.Index0Param2Write=true, Module.Index0Param3Write=true)
  ]
  particleUpdateScript [
   &ParticleState()
   &ScaleMeshSize(Scale Factor=&VectorFromCurve(VectorCurve=DataInterface() {propertyValues {
       XCurve {keys [<at 0.2 1 RCIM_Cubic>,<at 0.4 0.5 RCIM_Cubic>,<at 0.6 1 RCIM_Cubic>,<at 0.8 0.5 RCIM_Cubic>,<at 1 1 RCIM_Cubic>]}
       YCurve {keys [<at 0.2 1 RCIM_Cubic>,<at 0.4 0.5 RCIM_Cubic>,<at 0.6 1 RCIM_Cubic>,<at 0.8 0.5 RCIM_Cubic>,<at 1 1 RCIM_Cubic>]}
       ZCurve {keys [
         {arriveTangent=-0.0244, interpMode=RCIM_Cubic, leaveTangent=-0.0244, tangentMode=RCTM_User, time=0, value=3}
         {
          arriveTangent=-1.7224
          arriveTangentWeight=0.1414
          interpMode=RCIM_Cubic
          leaveTangent=-1.7224
          leaveTangentWeight=0.33
          tangentMode=RCTM_User
          time=0.2
          value=1
         }
         <at 1 0.5 RCIM_Cubic>
        ]}
      }}))
   &SolveForcesAndVelocity(Clamp Velocity=false, Limit Acceleration=false)
  ]
  rendererClasses [&NiagaraMeshRendererProperties]
  renderers [{rendererClass=&NiagaraMeshRendererProperties, rendererIndex=0}]
 }
 OrbitBlobs_Trails {
  emitterUpdateScript [&EmitterState()]
  particleSpawnScript [
   &InitializeParticle(
    Color=<rgba 0.01 0.01 0.01 1.5>
    Lifetime=0.5
    Ribbon Width=&RandomRangeFloat(Minimum=4, Maximum=6)
    Ribbon Width Mode=::NewEnumerator1
    UsePositionOffset=false
    Write Material Random=true
   )
   &AddVelocity(
    Cone Angle=3
    Cone Axis=<0 0 1>
    Module.UseConeFalloff=false
    Module.UseVelDistribution=false
    Override Seed=false
    Velocity Mode=::NewEnumerator2
    Velocity Speed=&Lerp_Float(
     A=200
     B=300
     Alpha=&FloatFromCurve(
      FloatCurve=DataInterface() {propertyValues {Curve {keys [
          {arriveTangent=1.5889, interpMode=RCIM_Cubic, leaveTangent=1.5889, tangentMode=RCTM_User, time=0, value=0}
          {interpMode=RCIM_Cubic, tangentMode=RCTM_User, time=1, value=1}
         ]}}}
      CurveIndex=$System.FN_Contrail_Setup.Dive Value
     )
    )
   )
  ]
  particleUpdateScript [
   &ParticleState()
   &SolveForcesAndVelocity(Clamp Velocity=false, Limit Acceleration=false)
   &ScaleRibbonWidth(Ribbon Width Scale=&Lerp_Float(
     A=1
     B=4
     Alpha=&FloatFromCurve(FloatCurve=DataInterface() {propertyValues {Curve {keys [
          {interpMode=RCIM_Cubic, tangentMode=RCTM_User, time=0, value=0}
          {arriveTangent=1.6527, interpMode=RCIM_Cubic, leaveTangent=1.6527, tangentMode=RCTM_User, time=1, value=1}
         ]}}})
    ))
   &ScaleColor(
    Scale Alpha=&Lerp_Float(
     A=0
     B=1
     Alpha=&FloatFromCurve(FloatCurve=DataInterface() {propertyValues {Curve {keys [
          {arriveTangent=-1.219, interpMode=RCIM_Cubic, leaveTangent=-1.219, tangentMode=RCTM_User, time=0, value=1}
          {arriveTangent=-0.0448, interpMode=RCIM_Cubic, leaveTangent=-0.0448, tangentMode=RCTM_User, time=1, value=0}
         ]}}})
    )
    ScaleA=true
    ScaleRGB=true
   )
   &DynamicMaterialParameters(
    Index 0 Param 1=&SelectFloatFromArray(
     Direct Array Index=&ReturnParticleIDAttributes(Particle ID Attribute=::NewEnumerator1, Niagara ID=$Particles.RibbonID)
     Float Selection Array=DataInterface() {propertyValues {FloatData [0,0.2,0.45,0.7]}}
    )
    Module.Index0Param0Write=true
    Module.Index0Param1Write=true
    Module.Index0Param2Write=true
    Module.Index0Param3Write=true
   )
  ]
  rendererClasses [&NiagaraRibbonRendererProperties]
  renderers [{rendererClass=&NiagaraRibbonRendererProperties}]
 }
 SpeedLines {
  emitterUpdateScript [
   &EmitterState()
   &SpawnRate(
    SpawnRate=&MakeCustomFloatFromBool(Bool=$User.bIsFrontEndPreview, True Float=0, False Float=31)
    Use Spawn Probability=false
   )
  ]
  particleSpawnScript [
   &InitializeParticle(
    Color=<rgba 1 1 1 0.05>
    Lifetime Max=1
    Lifetime Mode=::NewEnumerator1
    Sprite Size Max=<6 90>
    Sprite Size Min=<3 30>
    Sprite Size Mode=::NewEnumerator2
    UsePositionOffset=false
    Write Material Random=true
   )
   &ShapeLocation(
    Cylinder Height=400
    Cylinder Height Midpoint=1
    Cylinder Radius=600
    MaskSpawnGroup=false
    Offset=&Lerp_Vector(
     A=<0 0 700>
     B=<0 0 1400>
     Alpha=&FloatFromCurve(
      FloatCurve=DataInterface() {propertyValues {Curve {keys [
          {arriveTangent=2.5794, interpMode=RCIM_Cubic, leaveTangent=2.5794, tangentMode=RCTM_User, time=0, value=0}
          {interpMode=RCIM_Cubic, tangentMode=RCTM_User, time=1, value=1}
         ]}}}
      CurveIndex=$System.FN_Contrail_Setup.Dive Value
     )
    )
    Offset Mode=::NewEnumerator0
    Override Seed=false
    Rotation Axis=$System.FN_Contrail_Setup.Direction Vector
    Rotation Mode=::NewEnumerator1
    Shape Primitive=::NewEnumerator1
    Surface Only=true
    Surface Only Band Thickness=0.2
    Transform Order:ENiagara_TransformOrder=::NewEnumerator1
    Use Endcaps In Surface Only Mode=false
    UseOwnerScale=true
   )
  ]
  particleUpdateScript [
   &ParticleState()
   /SetVariables(Particles.NewAge:NiagaraFloat=&Multiply_Float(A=$Particles.Age, B=1)) {isSetParameters=true}
   /SetVariables(Particles.Velocity=&Multiply_VectorByFloat(
     Vector=&Lerp_Vector(A=$System.FN_Contrail_Setup.Direction Vector, B=$System.FN_Contrail_Setup.Direction Velocity, Alpha=$Particles.NewAge)
     Float=&Lerp_Float(
      A=2
      B=6
      Alpha=&FloatFromCurve(
       FloatCurve=DataInterface() {propertyValues {Curve {keys [
           {arriveTangent=0.8337, interpMode=RCIM_Cubic, leaveTangent=0.8337, tangentMode=RCTM_User, time=0, value=0}
           {interpMode=RCIM_Cubic, tangentMode=RCTM_User, time=1, value=1}
          ]}}}
       CurveIndex=$System.FN_Contrail_Setup.Dive Value
      )
     )
    )) {isSetParameters=true}
   /SetVariables(Particles.Position=&AddVectorToPosition(
     Position=$Particles.Initial.Position
     Vector=&Multiply_VectorByFloat(Vector=$Particles.Velocity, Float=$Particles.Age)
    )) {isSetParameters=true, module=SetVariables_#7}
   &ScaleSpriteSize(
    Scale Factor=&Lerp_Vector2(
     A=<0 0>
     B=<1 2>
     Alpha=&FloatFromCurve(FloatCurve=DataInterface() {propertyValues {Curve {keys [
          {arriveTangent=4.1785, interpMode=RCIM_Cubic, leaveTangent=4.1785, tangentMode=RCTM_User, time=0, value=0}
          {arriveTangent=1.1179, interpMode=RCIM_Cubic, leaveTangent=1.1179, tangentMode=RCTM_User, time=0.2042, value=0.6228}
          {arriveTangent=0.0224, interpMode=RCIM_Cubic, leaveTangent=0.0224, tangentMode=RCTM_User, time=1, value=1}
         ]}}})
    )
    Scale Sprite Size Mode=::NewEnumerator2
   )
   &ScaleColor(
    Scale Alpha=&FloatFromCurve(
     FloatCurve=DataInterface() {propertyValues {Curve {keys [
         {arriveTangent=9.0623, interpMode=RCIM_Cubic, leaveTangent=9.0623, tangentMode=RCTM_User, time=0, value=0}
         {interpMode=RCIM_Cubic, tangentMode=RCTM_User, time=0.2, value=1}
         {tangentMode=RCTM_User, time=1, value=0}
        ]}}}
     Scale Curve=$System.FN_Contrail_Setup.Normalized Intro Age
    )
    ScaleA=true
    ScaleRGB=false
   )
  ]
  rendererClasses [&NiagaraSpriteRendererProperties]
  renderers [{rendererClass=&NiagaraSpriteRendererProperties}]
 }
}
summary {
 systemName=NS_MissusMind_Contrail
 emitters [
  {emitterName=OrbitBlobs, rendererClasses [&NiagaraMeshRendererProperties]}
  {emitterName=SpeedLines, rendererClasses [&NiagaraSpriteRendererProperties]}
  {emitterName=OrbitBlobs_Trails, rendererClasses [&NiagaraRibbonRendererProperties]}
  {emitterName=OrbitBlobs_Drips, rendererClasses [&NiagaraMeshRendererProperties]}
  {bEnabled=false, emitterName=Hands_Trails, rendererClasses [&NiagaraRibbonRendererProperties]}
  {bEnabled=false, emitterName=Hands_Blobs, rendererClasses [&NiagaraMeshRendererProperties]}
  {bEnabled=false, emitterName=Hands_BlobDrips, rendererClasses [&NiagaraMeshRendererProperties]}
  {emitterName=OrbitBlobs001, rendererClasses [&NiagaraMeshRendererProperties]}
  {emitterName=FN_AttachedRibbon, rendererClasses [&NiagaraRibbonRendererProperties]}
 ]
 userVariables [
  User.bIsFrontEnd {type=NiagaraBool}
  User.bIsFrontEndPreview {type=NiagaraBool}
  User.ParaGlideLeanAlpha {type=NiagaraFloat, defaultValue=0, description=""}
  User.RibbonIDOffset {type=NiagaraInt32}
  User.SkeletalMesh {
   type=NiagaraDataInterfaceSkeletalMesh
   defaultValue {
    struct=NiagaraExt_VariableValue_DataInterface
    value {dataInterface=&self:NiagaraDataInterfaceSkeletalMesh_0}
   }
  }
 ]
}
